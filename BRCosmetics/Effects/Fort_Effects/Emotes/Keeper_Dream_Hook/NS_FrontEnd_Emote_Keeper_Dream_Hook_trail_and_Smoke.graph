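# Front-end Niagara effect for the Keeper Dream Hook emote, adding colored hand glows, stars, falling lead sparks, flashes and smoke.
# Lead sparks spawn at offset hand bones on User.SkeletalMesh, fall, and die on a collision plane set from a third sampled bone.
# They emit location and collision events carrying a color ID.
# Bursts use sprites, smoke mixes material-animated meshes with flipbook sprites, and several older emitter variants are disabled.
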NiagaraSystem /BRCosmetics/Effects/Fort_Effects/Emotes/Keeper_Dream_Hook/NS_FrontEnd_Emote_Keeper_Dream_Hook_trail_and_Smoke
dependencies {
 usedDataInterfaces [&NiagaraDataInterfaceCurve, &NiagaraDataInterfaceColorCurve, &NiagaraDataInterfaceVectorCurve]
 usedDynamicInputs []
 usedModules [&SystemState]
}
emitters {
 Burst_Flash {
  bEnabled=false
  emitterUpdateScript [&EmitterState(Life Cycle Mode=::NewEnumerator1, Loop Behavior=::NewEnumerator1, Loop Duration=3, UseLoopDelay=false)]
  particleSpawnScript [
   /Niagara/Modules/Spawn/Initialization/InitializeParticle(
    Color=&MakeCustomLinearColorFromBool(
     Bool=/Niagara/DynamicInputs/Random/RandomBool(Randomness Mode:ENiagaraRandomnessMode=::NewEnumerator0, Random Seed:NiagaraInt32=0, Fixed Random Seed:NiagaraBool=false)
     True Color=<rgba 1.3 0.4536 1.2601 1>
     False Color=&MakeCustomLinearColorFromBool(
      Bool=&RandomBool(Randomness Mode:ENiagaraRandomnessMode=::NewEnumerator0, Random Seed:NiagaraInt32=0, Fixed Random Seed:NiagaraBool=false)
      True Color=<rgba 0.3324 1.2277 2.5 1>
      False Color=<rgba 1 0.9106 0.4452 1>
     )
    )
    Lifetime=0.12
    Mesh Scale=&UniformRangedVector(Minimum=<0.6 0.5 0.5>, Maximum=<1.3 0.5 0.5>)
    Module.Write Color=true
    Module.Write Lifetime=true
    Module.Write Mass=true
    Module.Write Material Random=true
    Module.Write Position=true
    Module.Write Scale=true
    Module.Write SpriteRotation=false
    Module.Write SpriteSize=true
    Sprite Size=&Vector2DFromFloat(Value=250)
   )
   &AddVelocity(Velocity=<0 0 1>) {module=AddVelocity001}
   &DynamicMaterialParameters(
    Index 0 Param 1=&UniformRangedFloat(Minimum=0.1, Maximum=0.4)
    Index 0 Param 2=1
    Index 0 Param 3=1
    Index 0 Param 4=1
    Module.Param0WriteEnabled=true
    Module.Param1WriteEnabled=true
    Module.Param2WriteEnabled=true
    Module.Param3WriteEnabled=true
   )
   /NiagaraFN/Sprite/FN_SpriteAlignmentAndFacing(SpriteFacing:Vector3f=<0 0 1>) {module=SpriteAlignmentAndFacing}
  ]
  particleUpdateScript [
   &SolveForcesAndVelocity(Clamp Velocity=false, Limit Acceleration=false)
   &ParticleState()
   &ScaleColor(
    Scale Alpha=&FloatFromCurve(FloatCurve=DataInterface() {propertyValues {Curve {keys [<at 0 0>, {arriveTangent=1.2677, leaveTangent=1.2677, time=0.1349, value=1.0023}, <at 1 0>]}}})
    Scale RGB=&VectorFromCurve(VectorCurve=DataInterface() {propertyValues {XCurve {keys [<at 0 1>,<at 1 1>]}, YCurve {keys [<at 0 1>,<at 1 1>]}, ZCurve {keys [<at 0 1>,<at 1 1>]}}})
    ScaleA=true
    ScaleRGB=true
   )
  ]
  rendererClasses [&NiagaraSpriteRendererProperties]
  renderers [
   {rendererClass=&NiagaraSpriteRendererProperties}
   {rendererClass=&NiagaraSpriteRendererProperties, rendererIndex=1}
  ]
 }
 Burst_Flash001 {
  emitterUpdateScript [&EmitterState(Life Cycle Mode=::NewEnumerator1, Loop Behavior=::NewEnumerator1, Loop Duration=3, UseLoopDelay=false)]
  particleSpawnScript [
   &InitializeParticle(
    Color=&MakeCustomLinearColorFromBool(
     Bool=&RandomBool(Randomness Mode:ENiagaraRandomnessMode=::NewEnumerator0, Random Seed:NiagaraInt32=0, Fixed Random Seed:NiagaraBool=false)
     True Color=<rgba 1.3 0.4536 1.2601 1>
     False Color=&MakeCustomLinearColorFromBool(
      Bool=&RandomBool(Randomness Mode:ENiagaraRandomnessMode=::NewEnumerator0, Random Seed:NiagaraInt32=0, Fixed Random Seed:NiagaraBool=false)
      True Color=<rgba 0.3324 1.2277 2.5 1>
      False Color=<rgba 1 0.9106 0.4452 1>
     )
    )
    Lifetime=0.12
    Mesh Scale=&UniformRangedVector(Minimum=<0.6 0.5 0.5>, Maximum=<1.3 0.5 0.5>)
    Module.Write Color=true
    Module.Write Lifetime=true
    Module.Write Mass=true
    Module.Write Material Random=true
    Module.Write Position=true
    Module.Write Scale=true
    Module.Write SpriteRotation=false
    Module.Write SpriteSize=true
    Sprite Size=&UniformRangedVector2D(Minimum=<50 50>, Maximum=<80 80>)
   )
   &AddVelocity(Velocity=<0 0 1>) {module=AddVelocity001}
   &DynamicMaterialParameters(
    Index 0 Param 1=&UniformRangedFloat(Minimum=0.1, Maximum=0.4)
    Index 0 Param 2=1
    Index 0 Param 3=1
    Index 0 Param 4=1
    Module.Param0WriteEnabled=true
    Module.Param1WriteEnabled=true
    Module.Param2WriteEnabled=true
    Module.Param3WriteEnabled=true
   )
   &FN_SpriteAlignmentAndFacing(SpriteFacing:Vector3f=<0 0 1>) {module=SpriteAlignmentAndFacing}
  ]
  particleUpdateScript [
   &SolveForcesAndVelocity(Clamp Velocity=false, Limit Acceleration=false)
   &ParticleState()
   &ScaleSpriteSize(Scale Factor=&Vector2DFromFloat(Value=&FloatFromCurve(FloatCurve=DataInterface() {propertyValues {
        bOptimizeLUT=false
        Curve {keys [{arriveTangent=2.7713, interpMode=RCIM_Cubic, leaveTangent=2.7713, tangentMode=RCTM_User, time=0, value=1}, <at 1 3 RCIM_Cubic>]}
       }})))
   &ScaleColor(
    Scale Alpha=&FloatFromCurve(FloatCurve=DataInterface() {propertyValues {Curve {keys [<at 0 0>, {arriveTangent=1.2677, leaveTangent=1.2677, time=0.1349, value=1.0023}, <at 1 0>]}}})
    Scale RGB=&VectorFromCurve(VectorCurve=DataInterface() {propertyValues {XCurve {keys [<at 0 1>,<at 1 1>]}, YCurve {keys [<at 0 1>,<at 1 1>]}, ZCurve {keys [<at 0 1>,<at 1 1>]}}})
    ScaleA=true
    ScaleRGB=true
   )
  ]
  rendererClasses [&NiagaraSpriteRendererProperties]
  renderers [
   {rendererClass=&NiagaraSpriteRendererProperties}
   {rendererClass=&NiagaraSpriteRendererProperties, rendererIndex=1}
  ]
 }
 Burst_Sparkles {
  emitterUpdateScript [&EmitterState(Life Cycle Mode=::NewEnumerator1, Loop Behavior=::NewEnumerator1, Loop Duration=3, UseLoopDelay=false)]
  particleSpawnScript [
   &InitializeParticle(
    Color=&MakeCustomLinearColorFromBool(
     Bool=&RandomBool(Randomness Mode:ENiagaraRandomnessMode=::NewEnumerator0, Random Seed:NiagaraInt32=0, Fixed Random Seed:NiagaraBool=false)
     True Color=<rgba 1.3 0.4536 1.2601 1>
     False Color=&MakeCustomLinearColorFromBool(
      Bool=&RandomBool(Randomness Mode:ENiagaraRandomnessMode=::NewEnumerator0, Random Seed:NiagaraInt32=0, Fixed Random Seed:NiagaraBool=false)
      True Color=<rgba 0.3324 1.2277 2.5 1>
      False Color=<rgba 1 0.9106 0.4452 1>
     )
    )
    Lifetime=&UniformRangedFloat(Minimum=0.5, Maximum=1.7)
    Mesh Scale=&UniformRangedVector(Minimum=<0.6 0.5 0.5>, Maximum=<1.3 0.5 0.5>)
    Module.Write Color=true
    Module.Write Lifetime=true
    Module.Write Mass=true
    Module.Write Material Random=true
    Module.Write Position=true
    Module.Write Scale=false
    Module.Write SpriteRotation=false
    Module.Write SpriteSize=true
    Sprite Size=&Vector2DFromFloat(Value=&UniformRangedFloat(Minimum=3, Maximum=8))
   )
   &AddVelocity(Velocity=<0 0 1>) {module=AddVelocity001}
   &DynamicMaterialParameters(
    Index 0 Param 1=&UniformRangedFloat(Minimum=0.1, Maximum=0.4)
    Index 0 Param 2=1
    Index 0 Param 3=1
    Index 0 Param 4=1
    Module.Param0WriteEnabled=true
    Module.Param1WriteEnabled=true
    Module.Param2WriteEnabled=true
    Module.Param3WriteEnabled=true
   )
   &FN_SpriteAlignmentAndFacing(SpriteFacing:Vector3f=<0 0 1>) {module=SpriteAlignmentAndFacing}
  ]
  particleUpdateScript [
   &ParticleState()
   &ScaleSpriteSize(Scale Factor=&Vector2DFromFloat(Value=&FloatFromCurve(FloatCurve=DataInterface() {propertyValues {
        bOptimizeLUT=false
        Curve {keys [{arriveTangent=7.5169, interpMode=RCIM_Cubic, leaveTangent=7.5169, tangentMode=RCTM_User, time=0, value=1}, <at 1 0.5 RCIM_Cubic>]}
       }})))
   &ScaleColor(
    Scale Alpha=&FloatFromCurve(FloatCurve=DataInterface() {propertyValues {Curve {keys [<at 0 0>, {arriveTangent=1.2677, leaveTangent=1.2677, time=0.1349, value=1.0023}, <at 1 0>]}}})
    Scale RGB=&VectorFromCurve(VectorCurve=DataInterface() {propertyValues {XCurve {keys [<at 0 1>,<at 1 1>]}, YCurve {keys [<at 0 1>,<at 1 1>]}, ZCurve {keys [<at 0 1>,<at 1 1>]}}})
    ScaleA=true
    ScaleRGB=true
   )
   &Drag(Use Linear Drag=true, Use Rotational Drag=false)
   &SolveForcesAndVelocity(Clamp Velocity=false, Limit Acceleration=false)
  ]
  rendererClasses [&NiagaraSpriteRendererProperties]
  renderers [
   {rendererClass=&NiagaraSpriteRendererProperties}
   {rendererClass=&NiagaraSpriteRendererProperties, rendererIndex=1}
  ]
 }
 Flash_lead {
  bEnabled=false
  emitterUpdateScript [
   &EmitterState(Life Cycle Mode=::NewEnumerator1, Loop Behavior=::NewEnumerator1, Loop Duration=2, UseLoopDelay=false)
   &SpawnRate(SpawnRate=5, Use Spawn Probability=false)
   &SpawnRate(SpawnGroup=1, SpawnRate=5, Use Spawn Probability=false) {module=SpawnRate001}
  ]
  particleSpawnScript [
   &FN_SetSpawnGroup_v2(Number of Spawn Groups=3)
   /SetVariables(Particles.ColorID:NiagaraFloat=&Divide_Float(A=&MakeFloatFromInt(Int=$Particles.SpawnGroup), B=2)) {isSetParameters=true, module=SetVariables_#1}
   &InitializeParticle(
    Color=&ColorFromCurve(
     ColorCurve=DataInterface() {propertyValues {
       AlphaCurve {keys [<at 0 1 RCIM_Cubic>,<at 1 1 RCIM_Cubic>]}
       BlueCurve {keys [<at 0 0.9649 RCIM_Cubic>,<at 0.5 5>,<at 1 0.1146 RCIM_Cubic>]}
       GreenCurve {keys [<at 0 0.2552 RCIM_Cubic>,<at 0.5 0.7>,<at 1 0.7786 RCIM_Cubic>]}
       RedCurve {keys [<at 0 1 RCIM_Cubic>,<at 0.5 0>,<at 1 1 RCIM_Cubic>]}
      }}
     CurveIndex=$Particles.ColorID
    )
    Lifetime=&UniformRangedFloat(Minimum=1, Maximum=1.5)
    Module.Write Color=true
    Module.Write Lifetime=true
    Module.Write Mass=true
    Module.Write Material Random=true
    Module.Write Position=true
    Module.Write Scale=false
    Module.Write SpriteRotation=true
    Module.Write SpriteSize=true
    Sprite Rotation=&UniformRangedFloat(Minimum=0, Maximum=360)
    Sprite Size=&Vector2DFromFloat(Value=&UniformRangedFloat(Minimum=30, Maximum=50))
   )
   &AddVelocityFromPoint(Module.Use Falloff Distance=false, Velocity Strength=70)
   &SkeletalMeshLocation(
    Bone Sampling Mode=::NewEnumerator4
    Inherit Velocity=false
    "Inherit Velocity (Scale)"=0.5
    Interpolate Mesh Sampling=false
    MaskSpawnGroup=true
    Sampled Position Offset=<13 0 0>
    Sampled Position Offset Coordinate Space:ENiagaraMeshTransforms=::NewEnumerator3
    Skeletal Mesh=$User.SkeletalMesh
   )
   &SkeletalMeshLocation(
    Bone / Socket Index=1
    Bone Sampling Mode=::NewEnumerator4
    Inherit Velocity=false
    "Inherit Velocity (Scale)"=0.5
    Interpolate Mesh Sampling=false
    MaskSpawnGroup=true
    Sampled Position Offset=<-13 0 0>
    Sampled Position Offset Coordinate Space:ENiagaraMeshTransforms=::NewEnumerator3
    Skeletal Mesh=$User.SkeletalMesh
    Spawn Group Mask=1
   ) {module=SkeletalMeshLocation002}
  ]
  particleUpdateScript [
   &GravityForce(Gravity=<0 0 -500>)
   &ScaleColor(
    Scale Alpha=&FloatFromCurve(
     FloatCurve=DataInterface() {propertyValues {Curve {keys [<at 0 0>,<at 0.3 1>,<at 1 0>]}}}
     Scale Curve=&FN_FrontEndFloatModifier(Gameplay=0, Front End=0)
    )
    ScaleA=true
    ScaleRGB=true
   )
   &ParticleState()
   &Collision(
    Advanced Aging Rate=1
    Age Colliding Particles=false
    Analytical Collision Plane Position 1:Vector3f=<0 0 20>
    CPU Collision Type=::NewEnumerator3
    Friction=1
    Module.Control Rest State Via Surface Normal=false
    Module.EnableMaxCollisionCount=false
    Module.Randomize Collision Normal Vector=false
    Particle Radius Scale=0.1
    Perform 2 Analytical Collision Queries=false
    Restitution=1
   )
   &SolveForcesAndVelocity(Clamp Velocity=false, Limit Acceleration=false)
   &ScaleSpriteSize(Scale Factor=&Vector2DFromFloat(Value=&UniformRangedFloat(Minimum=0.3, Maximum=1)))
   &GenerateLocationEvent(Event Send Rate=30, Float to Send as Random Float:NiagaraFloat=$Particles.Initial.ColorID, Module.Use Event Send Rate=true)
   &KillParticles(Kill Particles=$Particles.HasCollided)
   /Niagara/Modules/Events/GenerateCollisionEvent(Module.Use Time Between Events=false, Module.Use Velocity Threshold=true, Random Value:NiagaraFloat=$Particles.Initial.ColorID)
  ]
  rendererClasses [&NiagaraSpriteRendererProperties]
  renderers [{rendererClass=&NiagaraSpriteRendererProperties}]
 }
 Flash_lead001 {
  emitterUpdateScript [
   &EmitterState(Life Cycle Mode=::NewEnumerator1, Loop Behavior=::NewEnumerator1, Loop Duration=2, UseLoopDelay=false)
   &SpawnRate(SpawnRate=5, Use Spawn Probability=false)
   &SpawnRate(SpawnGroup=1, SpawnRate=5, Use Spawn Probability=false) {module=SpawnRate001}
  ]
  particleSpawnScript [
   &FN_SetSpawnGroup_v2(Number of Spawn Groups=3)
   /SetVariables(Particles.ColorID:NiagaraFloat=&Divide_Float(A=&MakeFloatFromInt(Int=$Particles.SpawnGroup), B=2)) {isSetParameters=true, module=SetVariables_#1}
   &InitializeParticle(
    Color=&ColorFromCurve(
     ColorCurve=DataInterface() {propertyValues {
       AlphaCurve {keys [<at 0 1 RCIM_Cubic>,<at 1 1 RCIM_Cubic>]}
       BlueCurve {keys [<at 0 0.9649 RCIM_Cubic>,<at 0.5 5>,<at 1 0.1146 RCIM_Cubic>]}
       GreenCurve {keys [<at 0 0.2552 RCIM_Cubic>,<at 0.5 0.7>,<at 1 0.7786 RCIM_Cubic>]}
       RedCurve {keys [<at 0 1 RCIM_Cubic>,<at 0.5 0>,<at 1 1 RCIM_Cubic>]}
      }}
     CurveIndex=$Particles.ColorID
    )
    Lifetime=&UniformRangedFloat(Minimum=1, Maximum=1.5)
    Module.Write Color=true
    Module.Write Lifetime=true
    Module.Write Mass=true
    Module.Write Material Random=true
    Module.Write Position=true
    Module.Write Scale=false
    Module.Write SpriteRotation=true
    Module.Write SpriteSize=true
    Sprite Rotation=&UniformRangedFloat(Minimum=0, Maximum=360)
    Sprite Size=&Vector2DFromFloat(Value=&UniformRangedFloat(Minimum=30, Maximum=50))
   )
   &AddVelocityFromPoint(Module.Use Falloff Distance=false, Velocity Strength=70)
   &SkeletalMeshLocation(
    Bone Sampling Mode=::NewEnumerator4
    Inherit Velocity=false
    "Inherit Velocity (Scale)"=0.5
    Interpolate Mesh Sampling=false
    MaskSpawnGroup=true
    Sampled Position Offset=<13 0 0>
    Sampled Position Offset Coordinate Space:ENiagaraMeshTransforms=::NewEnumerator3
    Skeletal Mesh=$User.SkeletalMesh
   )
   &SkeletalMeshLocation(
    Bone / Socket Index=1
    Bone Sampling Mode=::NewEnumerator4
    Inherit Velocity=false
    "Inherit Velocity (Scale)"=0.5
    Interpolate Mesh Sampling=false
    MaskSpawnGroup=true
    Sampled Position Offset=<-13 0 0>
    Sampled Position Offset Coordinate Space:ENiagaraMeshTransforms=::NewEnumerator3
    Skeletal Mesh=$User.SkeletalMesh
    Spawn Group Mask=1
   ) {module=SkeletalMeshLocation002}
   &SkeletalMeshLocation(
    Apply To Particle Position=false
    Bone / Socket Index=2
    Bone Sampling Mode=::NewEnumerator4
    Inherit Velocity=false
    "Inherit Velocity (Scale)"=0.5
    Interpolate Mesh Sampling=false
    MaskSpawnGroup=true
    Sampled Position Offset=<-13 0 0>
    Sampled Position Offset Coordinate Space:ENiagaraMeshTransforms=::NewEnumerator3
    Skeletal Mesh=$User.SkeletalMesh
    Spawn Group Mask=1
   ) {module=SkeletalMeshLocation001}
   /SetVariables(Particles.CollisionVector:Vector3f=$Output.SkeletalMeshLocation001.SampledPosition) {isSetParameters=true}
  ]
  particleUpdateScript [
   &GravityForce(Gravity=<0 0 -500>)
   &ScaleColor(
    Scale Alpha=&FloatFromCurve(FloatCurve=DataInterface() {propertyValues {Curve {keys [<at 0 0>,<at 0.3 1>,<at 1 0>]}}})
    ScaleA=true
    ScaleRGB=true
   )
   &ParticleState()
   &Collision(
    Advanced Aging Rate=1
    Age Colliding Particles=false
    Analytical Collision Plane Position 1:Vector3f=$Particles.CollisionVector
    CPU Collision Type=::NewEnumerator3
    Friction=1
    Module.Control Rest State Via Surface Normal=false
    Module.EnableMaxCollisionCount=false
    Module.Randomize Collision Normal Vector=false
    Particle Radius Scale=0.1
    Perform 2 Analytical Collision Queries=false
    Restitution=1
   )
   &SolveForcesAndVelocity(Clamp Velocity=false, Limit Acceleration=false)
   &ScaleSpriteSize(Scale Factor=&Vector2DFromFloat(Value=&UniformRangedFloat(Minimum=0.3, Maximum=1)))
   &GenerateLocationEvent(Event Send Rate=30, Float to Send as Random Float:NiagaraFloat=$Particles.Initial.ColorID, Module.Use Event Send Rate=true)
   &KillParticles(Kill Particles=$Particles.HasCollided)
   &GenerateCollisionEvent(Module.Use Time Between Events=false, Module.Use Velocity Threshold=true, Random Value:NiagaraFloat=$Particles.Initial.ColorID)
  ]
  rendererClasses [&NiagaraSpriteRendererProperties]
  renderers [{rendererClass=&NiagaraSpriteRendererProperties}]
 }
 Glow_RightHand {
  emitterUpdateScript [
   &EmitterState(Life Cycle Mode=::NewEnumerator1, Loop Behavior=::NewEnumerator1, UseLoopDelay=false)
   &SpawnRate(SpawnRate=2, Use Spawn Probability=false)
  ]
  particleSpawnScript [&InitializeParticle(
    Color=<rgba 1 1 1 0.7>
    Lifetime=&UniformRangedFloat(Minimum=1, Maximum=1)
    Module.Write Color=true
    Module.Write Lifetime=true
    Module.Write Mass=true
    Module.Write Material Random=true
    Module.Write Position=true
    Module.Write Scale=false
    Module.Write SpriteRotation=true
    Module.Write SpriteSize=true
    Sprite Rotation=&UniformRangedFloat(Minimum=0, Maximum=360)
    Sprite Size=&Vector2DFromFloat(Value=&UniformRangedFloat(Minimum=60, Maximum=60))
   )]
  particleUpdateScript [
   &SkeletalMeshLocation(
    Bone / Socket Index=1
    Bone Sampling Mode=::NewEnumerator4
    Inherit Velocity=false
    Interpolate Mesh Sampling=false
    MaskSpawnGroup=false
    Sampled Position Offset=<-10 0 0>
    Sampled Position Offset Coordinate Space:ENiagaraMeshTransforms=::NewEnumerator3
    Skeletal Mesh=$User.SkeletalMesh
   )
   &ScaleColor(
    Scale Alpha=&FloatFromCurve(FloatCurve=DataInterface() {propertyValues {Curve {keys [<at 0 0 RCIM_Cubic>,<at 0.4349 0.995 RCIM_Cubic>,<at 1 0>]}}})
    Scale RGB=&Multiply_Vector(
     A=&MakeVectorFromLinearColorRGB(Color=&ColorFromCurve(ColorCurve=DataInterface() {propertyValues {
         AlphaCurve {keys [<at 0 1 RCIM_Cubic>,<at 1 1 RCIM_Cubic>]}
         BlueCurve {keys [<at 0 0.957 RCIM_Cubic>,<at 0.9599 1>]}
         GreenCurve {keys [<at 0 0.0885 RCIM_Cubic>,<at 0.9599 0.4839>]}
         RedCurve {keys [<at 0 1 RCIM_Cubic>,<at 0.9599 0>]}
        }}))
     B=&MakeVectorFromLinearColorRGB(Color=<rgba 1 1 1 1>)
    )
    ScaleA=true
    ScaleRGB=true
   )
   &ParticleState()
   &SolveForcesAndVelocity(Clamp Velocity=false, Limit Acceleration=false)
  ]
  rendererClasses [&NiagaraSpriteRendererProperties]
  renderers [{rendererClass=&NiagaraSpriteRendererProperties}]
 }
 Glow_leftHand {
  emitterUpdateScript [
   &EmitterState(Life Cycle Mode=::NewEnumerator1, Loop Behavior=::NewEnumerator1, UseLoopDelay=false)
   &SpawnRate(SpawnRate=2, Use Spawn Probability=false)
  ]
  particleSpawnScript [&InitializeParticle(
    Color=<rgba 1 1 1 0.7>
    Lifetime=&UniformRangedFloat(Minimum=1, Maximum=1)
    Module.Write Color=true
    Module.Write Lifetime=true
    Module.Write Mass=true
    Module.Write Material Random=true
    Module.Write Position=true
    Module.Write Scale=false
    Module.Write SpriteRotation=true
    Module.Write SpriteSize=true
    Sprite Rotation=&UniformRangedFloat(Minimum=0, Maximum=360)
    Sprite Size=&Vector2DFromFloat(Value=&UniformRangedFloat(Minimum=60, Maximum=60))
   )]
  particleUpdateScript [
   &SkeletalMeshLocation(
    Bone Sampling Mode=::NewEnumerator4
    Inherit Velocity=false
    Interpolate Mesh Sampling=false
    MaskSpawnGroup=false
    Sampled Position Offset=<10 0 0>
    Sampled Position Offset Coordinate Space:ENiagaraMeshTransforms=::NewEnumerator3
    Skeletal Mesh=$User.SkeletalMesh
   )
   &ScaleColor(
    Scale Alpha=&FloatFromCurve(FloatCurve=DataInterface() {propertyValues {Curve {keys [<at 0 0 RCIM_Cubic>,<at 0.4349 0.995 RCIM_Cubic>,<at 1 0>]}}})
    Scale RGB=&Multiply_Vector(
     A=&MakeVectorFromLinearColorRGB(Color=&ColorFromCurve(ColorCurve=DataInterface() {propertyValues {
         AlphaCurve {keys [<at 0 1 RCIM_Cubic>,<at 1 1 RCIM_Cubic>]}
         BlueCurve {keys [<at 0 0.957 RCIM_Cubic>,<at 0.9599 1>]}
         GreenCurve {keys [<at 0 0.0885 RCIM_Cubic>,<at 0.9599 0.4839>]}
         RedCurve {keys [<at 0 1 RCIM_Cubic>,<at 0.9599 0>]}
        }}))
     B=&MakeVectorFromLinearColorRGB(Color=<rgba 1 1 1 1>)
    )
    ScaleA=true
    ScaleRGB=true
   )
   &ParticleState()
   &SolveForcesAndVelocity(Clamp Velocity=false, Limit Acceleration=false)
  ]
  rendererClasses [&NiagaraSpriteRendererProperties]
  renderers [{rendererClass=&NiagaraSpriteRendererProperties}]
 }
 High_smoke {
  bEnabled=false
  emitterUpdateScript [&EmitterState(Life Cycle Mode=::NewEnumerator1, Loop Behavior=::NewEnumerator1, Loop Duration=3, UseLoopDelay=false)]
  particleSpawnScript [
   &InitializeParticle(
    Lifetime=&UniformRangedFloat(Minimum=0.7, Maximum=0.8)
    Mesh Scale=&UniformRangedVector(Minimum=<0.8 0.5 0.5>, Maximum=<1.3 0.5 0.5>)
    Module.Write Color=false
    Module.Write Lifetime=true
    Module.Write Mass=true
    Module.Write Material Random=true
    Module.Write Position=true
    Module.Write Scale=true
    Module.Write SpriteRotation=false
    Module.Write SpriteSize=true
    Sprite Size=&Vector2DFromFloat(Value=200)
   )
   /Niagara/Modules/Spawn/Orientation/V2/InitialMeshOrientation(Use Orientation Vector=true, Use Rotation Vector=true)
   &AddVelocity(Velocity=<0 0 1>) {module=AddVelocity001}
   &DynamicMaterialParameters(
    Index 0 Param 1=&UniformRangedFloat(Minimum=0.1, Maximum=0.4)
    Index 0 Param 2=2
    Index 0 Param 3=&UniformRangedFloat(Minimum=0.4, Maximum=0.6)
    Index 0 Param 4=1
    Module.Param0WriteEnabled=true
    Module.Param1WriteEnabled=true
    Module.Param2WriteEnabled=true
    Module.Param3WriteEnabled=false
   )
  ]
  particleUpdateScript [
   &DynamicMaterialParameters(
    Index 0 Param 1=&UniformRangedFloat(Minimum=0.05, Maximum=0.25)
    Index 0 Param 2=&FloatFromCurve(FloatCurve=DataInterface() {propertyValues {Curve {keys [<at 0 1>,<at 1 1>]}}})
    Index 0 Param 3=&UniformRangedFloat(Minimum=0.5, Maximum=0.7)
    Index 0 Param 4=&FloatFromCurve(FloatCurve=DataInterface() {propertyValues {Curve {keys [
         {interpMode=RCIM_Cubic, tangentMode=RCTM_Break, time=0.25, value=1}
         {arriveTangent=-1.1387, interpMode=RCIM_Cubic, tangentMode=RCTM_Break, time=1, value=0}
        ]}}})
    Module.Param0WriteEnabled=false
    Module.Param1WriteEnabled=false
    Module.Param2WriteEnabled=false
    Module.Param3WriteEnabled=true
   ) {module=DynamicMaterialParameters001}
   &SolveForcesAndVelocity(Clamp Velocity=false, Limit Acceleration=false)
   &ParticleState()
   &ScaleColor(
    Color Value To Scale=$Particles.Spawn_Color
    Scale Alpha=&FloatFromCurve(FloatCurve=DataInterface() {propertyValues {Curve {keys [<at 0 0>, {arriveTangent=-2.6768, interpMode=RCIM_Cubic, leaveTangent=-2.6768, tangentMode=RCTM_User, time=0.1349, value=1.2}, <at 1 0>]}}})
    Scale RGB=&MakeVectorFromLinearColorRGB(Color=<rgba 1.5 1.5 1.5 1>)
    ScaleA=true
    ScaleRGB=true
   )
   &ScaleMeshSize(Scale Factor=&VectorFromCurve(VectorCurve=DataInterface() {propertyValues {
       XCurve {keys [
         <at 0 0 RCIM_Cubic>
         {arriveTangent=0.9145, interpMode=RCIM_Cubic, leaveTangent=0.9145, tangentMode=RCTM_User, time=0.1, value=0.7}
         <at 1 1>
        ]}
       YCurve {keys [<at 0 1>,<at 1 1>]}
       ZCurve {keys [<at 0 1>,<at 1 1>]}
      }}))
  ]
  rendererClasses [&NiagaraMeshRendererProperties]
  renderers [{rendererClass=&NiagaraMeshRendererProperties, rendererIndex=0}]
 }
 High_smoke001 {
  emitterUpdateScript [&EmitterState(Life Cycle Mode=::NewEnumerator1, Loop Behavior=::NewEnumerator1, Loop Duration=3, UseLoopDelay=false)]
  particleSpawnScript [
   &InitializeParticle(
    Lifetime=&UniformRangedFloat(Minimum=0.9, Maximum=1.6)
    Mesh Scale=&UniformRangedVector(Minimum=<0.8 0.5 0.5>, Maximum=<1 0.5 0.5>)
    Module.Write Color=false
    Module.Write Lifetime=true
    Module.Write Mass=true
    Module.Write Material Random=true
    Module.Write Position=true
    Module.Write Scale=true
    Module.Write SpriteRotation=false
    Module.Write SpriteSize=true
    Sprite Size=&Vector2DFromFloat(Value=200)
   )
   &InitialMeshOrientation(Use Orientation Vector=true, Use Rotation Vector=true)
   &AddVelocity(Velocity=<0 0 1>) {module=AddVelocity001}
   &DynamicMaterialParameters(
    Index 0 Param 1=&UniformRangedFloat(Minimum=0.15, Maximum=0.2)
    Index 0 Param 3=&UniformRangedFloat(Minimum=0.18, Maximum=0.25)
    Module.Param0WriteEnabled=true
    Module.Param1WriteEnabled=false
    Module.Param2WriteEnabled=true
    Module.Param3WriteEnabled=false
   )
  ]
  particleUpdateScript [
   &DynamicMaterialParameters(
    Index 0 Param 1=&UniformRangedFloat(Minimum=0.05, Maximum=0.25)
    Index 0 Param 2=&FloatFromCurve(FloatCurve=DataInterface() {propertyValues {Curve {keys [<at 0 1>,<at 1 1>]}}})
    Index 0 Param 3=&UniformRangedFloat(Minimum=0.5, Maximum=0.7)
    Index 0 Param 4=&FloatFromCurve(FloatCurve=DataInterface() {propertyValues {Curve {keys [
         {interpMode=RCIM_Cubic, tangentMode=RCTM_Break, time=0.4982, value=0.8}
         {arriveTangent=-1.5333, interpMode=RCIM_Cubic, leaveTangent=-1.5333, tangentMode=RCTM_User, time=0.6856, value=0.7182}
         <at 1 0 RCIM_Cubic>
        ]}}})
    Module.Param0WriteEnabled=false
    Module.Param1WriteEnabled=true
    Module.Param2WriteEnabled=false
    Module.Param3WriteEnabled=true
   ) {module=DynamicMaterialParameters001}
   &SolveForcesAndVelocity(Clamp Velocity=false, Limit Acceleration=false)
   &ParticleState()
   &ScaleColor(
    Color Value To Scale=$Particles.Spawn_Color
    Scale Alpha=&FloatFromCurve(
     FloatCurve=DataInterface() {propertyValues {Curve {keys [
         {arriveTangent=10.8313, interpMode=RCIM_Cubic, leaveTangent=10.8313, tangentMode=RCTM_User, time=0, value=0}
         {arriveTangent=-0.1108, interpMode=RCIM_Cubic, leaveTangent=-0.1108, tangentMode=RCTM_User, time=0.1349, value=1}
         <at 0.9848 0.8508>
        ]}}}
     Scale Curve=0.8
    )
    Scale RGB=&VectorFromCurve(VectorCurve=DataInterface() {propertyValues {XCurve {keys [<at 0 1>,<at 1 1>]}, YCurve {keys [<at 0 1>,<at 1 1>]}, ZCurve {keys [<at 0 1>,<at 1 1>]}}})
    ScaleA=true
    ScaleRGB=true
   )
   &ScaleMeshSize(Scale Factor=&VectorFromCurve(VectorCurve=DataInterface() {propertyValues {
       XCurve {keys [
         {arriveTangent=8.3601, interpMode=RCIM_Cubic, leaveTangent=8.3601, tangentMode=RCTM_User, time=0, value=0}
         {arriveTangent=1.3653, interpMode=RCIM_Cubic, leaveTangent=1.3653, tangentMode=RCTM_User, time=0.0939, value=0.3423}
         <at 1 1>
        ]}
       YCurve {keys [<at 0 1>,<at 1 1>]}
       ZCurve {keys [<at 0 1>,<at 1 1>]}
      }}))
  ]
  rendererClasses [&NiagaraMeshRendererProperties]
  renderers [{rendererClass=&NiagaraMeshRendererProperties, rendererIndex=0}]
 }
 High_smoke_shape_more {
  bEnabled=false
  emitterUpdateScript [&EmitterState(Life Cycle Mode=::NewEnumerator1, Loop Behavior=::NewEnumerator1, Loop Duration=3, UseLoopDelay=false)]
  particleSpawnScript [
   &InitializeParticle(
    Lifetime=&UniformRangedFloat(Minimum=0.7, Maximum=0.9)
    Mesh Scale=&UniformRangedVector(Minimum=<0.7 0.5 0.5>, Maximum=<1.5 0.5 0.5>)
    Module.Write Color=false
    Module.Write Lifetime=true
    Module.Write Mass=true
    Module.Write Material Random=true
    Module.Write Position=true
    Module.Write Scale=true
    Module.Write SpriteRotation=false
    Module.Write SpriteSize=true
    Sprite Size=&Vector2DFromFloat(Value=200)
   )
   &InitialMeshOrientation(Use Orientation Vector=true, Use Rotation Vector=true)
   &AddVelocity(Velocity=<0 0 1>) {module=AddVelocity001}
   &DynamicMaterialParameters(
    Index 0 Param 1=&UniformRangedFloat(Minimum=0.1, Maximum=0.4)
    Index 0 Param 3=&UniformRangedFloat(Minimum=0.5, Maximum=0.6)
    Module.Param0WriteEnabled=true
    Module.Param1WriteEnabled=false
    Module.Param2WriteEnabled=true
    Module.Param3WriteEnabled=false
   )
  ]
  particleUpdateScript [
   &DynamicMaterialParameters(
    Index 0 Param 1=&UniformRangedFloat(Minimum=0.05, Maximum=0.25)
    Index 0 Param 2=&FloatFromCurve(FloatCurve=DataInterface() {propertyValues {Curve {keys [<at 0 1>,<at 1 1>]}}})
    Index 0 Param 3=&UniformRangedFloat(Minimum=0.5, Maximum=0.7)
    Index 0 Param 4=&FloatFromCurve(FloatCurve=DataInterface() {propertyValues {Curve {keys [
         {interpMode=RCIM_Cubic, tangentMode=RCTM_Break, time=0.25, value=1}
         {arriveTangent=-1.1387, interpMode=RCIM_Cubic, tangentMode=RCTM_Break, time=1, value=0}
        ]}}})
    Module.Param0WriteEnabled=false
    Module.Param1WriteEnabled=true
    Module.Param2WriteEnabled=false
    Module.Param3WriteEnabled=true
   ) {module=DynamicMaterialParameters001}
   &SolveForcesAndVelocity(Clamp Velocity=false, Limit Acceleration=false)
   &ParticleState()
   &ScaleColor(
    Color Value To Scale=$Particles.Spawn_Color
    Scale Alpha=&FloatFromCurve(FloatCurve=DataInterface() {propertyValues {Curve {keys [<at 0 0>, {arriveTangent=-1.6695, interpMode=RCIM_Cubic, leaveTangent=-1.6695, tangentMode=RCTM_User, time=0.1349, value=0.7}, <at 1 0>]}}})
    Scale RGB=&MakeVectorFromLinearColorRGB(Color=<rgba 1.5 1.5 1.5 1>)
    ScaleA=true
    ScaleRGB=true
   )
   &ScaleMeshSize(Scale Factor=&VectorFromCurve(VectorCurve=DataInterface() {propertyValues {
       XCurve {keys [
         <at 0 0 RCIM_Cubic>
         {arriveTangent=0.9145, interpMode=RCIM_Cubic, leaveTangent=0.9145, tangentMode=RCTM_User, time=0.1, value=0.7}
         <at 1 1>
        ]}
       YCurve {keys [<at 0 1>,<at 1 1>]}
       ZCurve {keys [<at 0 1>,<at 1 1>]}
      }}))
  ]
  rendererClasses [&NiagaraMeshRendererProperties]
  renderers [{rendererClass=&NiagaraMeshRendererProperties, rendererIndex=0}]
 }
 High_smoke_shape_more001 {
  emitterUpdateScript [&EmitterState(Life Cycle Mode=::NewEnumerator1, Loop Behavior=::NewEnumerator1, Loop Duration=3, UseLoopDelay=false)]
  particleSpawnScript [
   &InitializeParticle(
    Lifetime=&UniformRangedFloat(Minimum=0.8, Maximum=1)
    Mesh Scale=&UniformRangedVector(Minimum=<0.6 0.4 0.4>, Maximum=<0.8 0.6 0.6>)
    Module.Write Color=false
    Module.Write Lifetime=true
    Module.Write Mass=true
    Module.Write Material Random=true
    Module.Write Position=true
    Module.Write Scale=true
    Module.Write SpriteRotation=false
    Module.Write SpriteSize=true
    Sprite Size=&Vector2DFromFloat(Value=200)
   )
   &InitialMeshOrientation(Use Orientation Vector=true, Use Rotation Vector=true)
   &AddVelocity(Velocity=<0 0 1>) {module=AddVelocity001}
   &DynamicMaterialParameters(
    Index 0 Param 1=&UniformRangedFloat(Minimum=0.15, Maximum=0.23)
    Index 0 Param 3=&UniformRangedFloat(Minimum=0.18, Maximum=0.25)
    Module.Param0WriteEnabled=true
    Module.Param1WriteEnabled=false
    Module.Param2WriteEnabled=true
    Module.Param3WriteEnabled=false
   )
  ]
  particleUpdateScript [
   &DynamicMaterialParameters(
    Index 0 Param 1=&UniformRangedFloat(Minimum=0.05, Maximum=0.25)
    Index 0 Param 2=&FloatFromCurve(FloatCurve=DataInterface() {propertyValues {Curve {keys [<at 0 1>,<at 1 1>]}}})
    Index 0 Param 3=&UniformRangedFloat(Minimum=0.5, Maximum=0.7)
    Index 0 Param 4=&FloatFromCurve(FloatCurve=DataInterface() {propertyValues {Curve {keys [
         {interpMode=RCIM_Cubic, tangentMode=RCTM_Break, time=0.25, value=0.9}
         {arriveTangent=-1.1387, interpMode=RCIM_Cubic, tangentMode=RCTM_Break, time=1, value=0}
        ]}}})
    Module.Param0WriteEnabled=false
    Module.Param1WriteEnabled=true
    Module.Param2WriteEnabled=false
    Module.Param3WriteEnabled=true
   ) {module=DynamicMaterialParameters001}
   &SolveForcesAndVelocity(Clamp Velocity=false, Limit Acceleration=false)
   &ParticleState()
   &ScaleColor(
    Color Value To Scale=$Particles.Spawn_Color
    Scale Alpha=&FloatFromCurve(
     FloatCurve=DataInterface() {propertyValues {Curve {keys [
         <at 0 0>
         {arriveTangent=-0.5011, interpMode=RCIM_Cubic, leaveTangent=-0.5011, tangentMode=RCTM_User, time=0.1349, value=0.7}
         {arriveTangent=-1.1835, interpMode=RCIM_Cubic, leaveTangent=-1.1835, tangentMode=RCTM_User, time=0.4339, value=0.4321}
         <at 1 0>
        ]}}}
     Scale Curve=0.9
    )
    Scale RGB=&VectorFromCurve(VectorCurve=DataInterface() {propertyValues {XCurve {keys [<at 0 1>,<at 1 1>]}, YCurve {keys [<at 0 1>,<at 1 1>]}, ZCurve {keys [<at 0 1>,<at 1 1>]}}})
    ScaleA=true
    ScaleRGB=true
   )
   &ScaleMeshSize(Scale Factor=&VectorFromCurve(VectorCurve=DataInterface() {propertyValues {
       XCurve {keys [
         {arriveTangent=4.1829, interpMode=RCIM_Cubic, leaveTangent=4.1829, tangentMode=RCTM_User, time=0, value=0}
         {arriveTangent=0.9145, interpMode=RCIM_Cubic, leaveTangent=0.9145, tangentMode=RCTM_User, time=0.1882, value=0.5338}
         <at 1 1>
        ]}
       YCurve {keys [<at 0 1>,<at 1 1>]}
       ZCurve {keys [<at 0 1>,<at 1 1>]}
      }}))
  ]
  rendererClasses [&NiagaraMeshRendererProperties]
  renderers [{rendererClass=&NiagaraMeshRendererProperties, rendererIndex=0}]
 }
 Round_smoke {
  bEnabled=false
  emitterUpdateScript [&EmitterState(Life Cycle Mode=::NewEnumerator1, Loop Behavior=::NewEnumerator1, Loop Delay=2, Loop Duration=3, UseLoopDelay=false)]
  particleSpawnScript [
   &FN_SetSpawnGroup_v2(Number of Spawn Groups=3)
   /SetVariables(Particles.ColorID:NiagaraFloat=&Divide_Float(A=&MakeFloatFromInt(Int=$Particles.SpawnGroup), B=2)) {isSetParameters=true, module=SetVariables_#1}
   &InitializeParticle(
    Color=&ColorFromCurve(
     ColorCurve=DataInterface() {propertyValues {
       AlphaCurve {keys [<at 0 1 RCIM_Cubic>,<at 1 1 RCIM_Cubic>]}
       BlueCurve {keys [<at 0 0.9548 RCIM_Cubic>,<at 0.5 5>,<at 1 0.0521 RCIM_Cubic>]}
       GreenCurve {keys [<at 0 0.0104 RCIM_Cubic>,<at 0.5 0.7>,<at 1 0.763 RCIM_Cubic>]}
       RedCurve {keys [<at 0 1 RCIM_Cubic>,<at 0.5 0>,<at 1 1 RCIM_Cubic>]}
      }}
     CurveIndex=$Particles.ColorID
    )
    Lifetime=&UniformRangedFloat(Minimum=1.4, Maximum=1.5)
    Mesh Scale=&UniformRangedVector(Minimum=<1 0.5 0.5>, Maximum=<1.2 0.5 0.5>)
    Module.Write Color=false
    Module.Write Lifetime=true
    Module.Write Mass=true
    Module.Write Material Random=true
    Module.Write Position=true
    Module.Write Scale=true
    Module.Write SpriteRotation=true
    Module.Write SpriteSize=true
    Sprite Rotation=&UniformRangedFloat(Minimum=0, Maximum=360)
    Sprite Size=&Vector2DFromFloat(Value=&UniformRangedFloat(Minimum=60, Maximum=70))
   )
  ]
  particleUpdateScript [
   &DynamicMaterialParameters(
    Index 0 Param 1=&FloatFromCurve(FloatCurve=DataInterface() {propertyValues {Curve {keys [<at 0 1>,<at 0.5 0.8>,<at 1 -0.1>]}}})
    Index 0 Param 4=1
    Module.Param0WriteEnabled=true
    Module.Param1WriteEnabled=true
    Module.Param2WriteEnabled=true
    Module.Param3WriteEnabled=true
   ) {module=DynamicMaterialParameters001}
   &Drag(Drag=&UniformRangedFloat(Minimum=2, Maximum=4), Use Linear Drag=true, Use Rotational Drag=false)
   &SolveForcesAndVelocity(Clamp Velocity=false, Limit Acceleration=false)
   &ParticleState()
   &ScaleColor(
    Color Value To Scale=$Particles.Spawn_Color
    Scale Alpha=&Multiply_Float(
     A=&FloatFromCurve(FloatCurve=DataInterface() {propertyValues {Curve {keys [<at 0 0>,<at 0.1 0.3>,<at 1 0>]}}})
     B=1
    )
    Scale RGB=<2 2 2>
    ScaleA=true
    ScaleRGB=true
   )
   &SubUVAnimation(End Frame=40, Start Frame=18)
   &ScaleSpriteSize(Scale Factor=&Vector2DFromFloat(Value=&FloatFromCurve(FloatCurve=DataInterface() {propertyValues {Curve {keys [{arriveTangent=-3.7814, time=0, value=1}, <at 0.9929 0.5>]}}})))
   &CameraOffset(Camera Offset Amount=35)
  ]
  rendererClasses [&NiagaraSpriteRendererProperties]
  renderers [{rendererClass=&NiagaraSpriteRendererProperties}]
 }
 Round_smoke001 {
  emitterUpdateScript [&EmitterState(Life Cycle Mode=::NewEnumerator1, Loop Behavior=::NewEnumerator1, Loop Delay=2, Loop Duration=3, UseLoopDelay=false)]
  particleSpawnScript [
   &FN_SetSpawnGroup_v2(Number of Spawn Groups=3)
   /SetVariables(Particles.ColorID:NiagaraFloat=&Divide_Float(A=&MakeFloatFromInt(Int=$Particles.SpawnGroup), B=2)) {isSetParameters=true, module=SetVariables_#1}
   &InitializeParticle(
    Color=&ColorFromCurve(
     ColorCurve=DataInterface() {propertyValues {
       AlphaCurve {keys [<at 0 1 RCIM_Cubic>,<at 1 1 RCIM_Cubic>]}
       BlueCurve {keys [<at 0 0.9548 RCIM_Cubic>,<at 0.5 5>,<at 1 0.0521 RCIM_Cubic>]}
       GreenCurve {keys [<at 0 0.0104 RCIM_Cubic>,<at 0.5 0.7>,<at 1 0.763 RCIM_Cubic>]}
       RedCurve {keys [<at 0 1 RCIM_Cubic>,<at 0.5 0>,<at 1 1 RCIM_Cubic>]}
      }}
     CurveIndex=$Particles.ColorID
    )
    Lifetime=&UniformRangedFloat(Minimum=1.5, Maximum=1.6)
    Mesh Scale=&UniformRangedVector(Minimum=<1 0.5 0.5>, Maximum=<1.2 0.5 0.5>)
    Module.Write Color=false
    Module.Write Lifetime=true
    Module.Write Mass=true
    Module.Write Material Random=true
    Module.Write Position=true
    Module.Write Scale=true
    Module.Write SpriteRotation=true
    Module.Write SpriteSize=true
    Sprite Rotation=&UniformRangedFloat(Minimum=0, Maximum=360)
    Sprite Size=&Vector2DFromFloat(Value=&UniformRangedFloat(Minimum=30, Maximum=35))
   )
  ]
  particleUpdateScript [
   &DynamicMaterialParameters(
    Index 0 Param 1=&FloatFromCurve(FloatCurve=DataInterface() {propertyValues {Curve {keys [<at 0 1>,<at 0.5 0.8>,<at 1 -0.1>]}}})
    Index 0 Param 4=1
    Module.Param0WriteEnabled=true
    Module.Param1WriteEnabled=true
    Module.Param2WriteEnabled=true
    Module.Param3WriteEnabled=true
   ) {module=DynamicMaterialParameters001}
   &ParticleState()
   &ScaleColor(
    Color Value To Scale=$Particles.Spawn_Color
    Scale Alpha=&Multiply_Float(
     A=&FloatFromCurve(FloatCurve=DataInterface() {propertyValues {Curve {keys [
          {arriveTangent=3.2558, interpMode=RCIM_Cubic, leaveTangent=3.2558, tangentMode=RCTM_User, time=0, value=0}
          <at 0.1 0.2 RCIM_Cubic>
          <at 1 0>
         ]}}})
     B=1
    )
    ScaleA=true
    ScaleRGB=true
   )
   &SubUVAnimation(End Frame=16, Start Frame=12)
   &ScaleSpriteSize(Scale Factor=&Vector2DFromFloat(Value=&FloatFromCurve(FloatCurve=DataInterface() {propertyValues {Curve {keys [{arriveTangent=-4.8823, interpMode=RCIM_Cubic, leaveTangent=1.9773, tangentMode=RCTM_User, time=0, value=1}, <at 1 2.25 RCIM_Cubic>]}}})))
   &CameraOffset(Camera Offset Amount=35)
   &Drag(Drag=&UniformRangedFloat(Minimum=3, Maximum=4), Use Linear Drag=true, Use Rotational Drag=false)
   &SolveForcesAndVelocity(Clamp Velocity=false, Limit Acceleration=false)
  ]
  rendererClasses [&NiagaraSpriteRendererProperties]
  renderers [{rendererClass=&NiagaraSpriteRendererProperties}]
 }
 Star_hand {
  emitterUpdateScript [
   &EmitterState(Life Cycle Mode=::NewEnumerator1, Loop Behavior=::NewEnumerator1, Loop Duration=2, UseLoopDelay=false)
   &SpawnRate(SpawnRate=10, Use Spawn Probability=false)
   &SpawnRate(SpawnGroup=1, SpawnRate=10, Use Spawn Probability=false) {module=SpawnRate001}
  ]
  particleSpawnScript [
   &InitializeParticle(
    Color=&MakeCustomLinearColorFromBool(
     Bool=&RandomBool(Randomness Mode:ENiagaraRandomnessMode=::NewEnumerator0, Random Seed:NiagaraInt32=0, Fixed Random Seed:NiagaraBool=false)
     True Color=<rgba 1 0.1406 0.9595 1>
     False Color=&MakeCustomLinearColorFromBool(
      Bool=&RandomBool(Randomness Mode:ENiagaraRandomnessMode=::NewEnumerator0, Random Seed:NiagaraInt32=0, Fixed Random Seed:NiagaraBool=false)
      True Color=<rgba 0.019 0.3135 1 1>
      False Color=<rgba 0.5 0.4121 0.1484 1>
     )
    )
    Lifetime=&UniformRangedFloat(Minimum=0.3, Maximum=0.5)
    Module.Write Color=true
    Module.Write Lifetime=true
    Module.Write Mass=true
    Module.Write Material Random=true
    Module.Write Position=true
    Module.Write Scale=false
    Module.Write SpriteRotation=true
    Module.Write SpriteSize=true
    Sprite Rotation=&UniformRangedFloat(Minimum=0, Maximum=360)
    Sprite Size=&Vector2DFromFloat(Value=&UniformRangedFloat(Minimum=15, Maximum=10))
   )
   &AddVelocityFromPoint(Module.Use Falloff Distance=false, Velocity Strength=10)
   &SkeletalMeshLocation(
    Bone Sampling Mode=::NewEnumerator4
    Inherit Velocity=false
    Interpolate Mesh Sampling=false
    MaskSpawnGroup=true
    Sampled Position Offset=<13 0 0>
    Sampled Position Offset Coordinate Space:ENiagaraMeshTransforms=::NewEnumerator3
    Skeletal Mesh=$User.SkeletalMesh
   )
   &SkeletalMeshLocation(
    Bone / Socket Index=1
    Bone Sampling Mode=::NewEnumerator4
    Inherit Velocity=false
    Interpolate Mesh Sampling=false
    MaskSpawnGroup=true
    Sampled Position Offset=<-13 0 0>
    Sampled Position Offset Coordinate Space:ENiagaraMeshTransforms=::NewEnumerator3
    Skeletal Mesh=$User.SkeletalMesh
    Spawn Group Mask=1
   ) {module=SkeletalMeshLocation001}
  ]
  particleUpdateScript [
   &ScaleColor(
    Scale Alpha=&FloatFromCurve(FloatCurve=DataInterface() {propertyValues {Curve {keys [<at 0 0>,<at 0.3 1>,<at 1 0>]}}})
    Scale RGB=&Multiply_Vector(
     A=&MakeVectorFromLinearColorRGB(Color=&ColorFromCurve(ColorCurve=DataInterface() {propertyValues {
         AlphaCurve {keys [<at 0 1 RCIM_Cubic>,<at 1 1 RCIM_Cubic>]}
         BlueCurve {keys [<at 0 3 RCIM_Cubic>,<at 1 1 RCIM_Cubic>]}
         GreenCurve {keys [<at 0 3 RCIM_Cubic>,<at 1 1 RCIM_Cubic>]}
         RedCurve {keys [<at 0 3 RCIM_Cubic>,<at 1 1 RCIM_Cubic>]}
        }}))
     B=&MakeVectorFromLinearColorRGB(Color=<rgba 1 1 1 1>)
    )
    ScaleA=true
    ScaleRGB=true
   )
   &ParticleState()
   &GravityForce(Gravity=<0 0 -500>)
   &SolveForcesAndVelocity(Clamp Velocity=false, Limit Acceleration=false)
   &ScaleSpriteSize(Scale Factor=&Vector2DFromFloat(Value=&UniformRangedFloat(Minimum=0, Maximum=1)))
  ]
  rendererClasses [&NiagaraSpriteRendererProperties]
  renderers [
   {rendererClass=&NiagaraSpriteRendererProperties}
   {rendererClass=&NiagaraSpriteRendererProperties, rendererIndex=1}
  ]
 }
 Trail {
  bEnabled=false
  emitterUpdateScript [&EmitterState(Life Cycle Mode=::NewEnumerator1, Loop Behavior=::NewEnumerator1, Loop Duration=2.5, UseLoopDelay=false)]
  particleSpawnScript [
   &InitializeParticle(
    Color=<rgba 15 15 15 1>
    Lifetime=0.5
    Module.Write Color=true
    Module.Write Lifetime=true
    Module.Write Mass=true
    Module.Write Material Random=true
    Module.Write Position=true
    Module.Write Scale=false
    Module.Write SpriteRotation=false
    Module.Write SpriteSize=true
    Position=&Add_Vector(A=/Niagara/DynamicInputs/Helpers/SimulationPosition(), B=<0.5 0 0>)
    Sprite Size=&Vector2DFromFloat(Value=20)
   ) {enabled=false}
   &InitializeRibbon(
    Lifetime=&UniformRangedFloat(Minimum=0.3, Maximum=0.4)
    Module.Write Color=false
    Module.Write Lifetime=true
    Module.Write Mass=true
    Module.Write Material Random=true
    Module.Write Position=true
    Module.Write RibbonTwist=false
    Module.Write RibbonWidth=true
    Ribbon Width=10
   )
  ]
  particleUpdateScript [
   &SolveForcesAndVelocity(Clamp Velocity=false, Limit Acceleration=false)
   &ParticleState()
   &ScaleRibbonWidth(Ribbon Width Scale=&FloatFromCurve(FloatCurve=DataInterface() {propertyValues {Curve {keys [<at 0 1>,<at 1 0>]}}}))
   &ScaleColor(
    Color Value To Scale=$Particles.Spawn_Color
    Scale Alpha=&FloatFromCurve(FloatCurve=DataInterface() {propertyValues {Curve {keys [<at 0 0.5>,<at 1 0>]}}})
    Scale RGB=&Multiply_Vector(
     A=&MakeVectorFromLinearColorRGB(Color=&ColorFromCurve(ColorCurve=DataInterface() {propertyValues {
         AlphaCurve {keys [<at 0 1 RCIM_Cubic>,<at 1 1 RCIM_Cubic>]}
         BlueCurve {keys [<at 0 1 RCIM_Cubic>,<at 1 1 RCIM_Cubic>]}
         GreenCurve {keys [<at 0 1 RCIM_Cubic>,<at 1 1 RCIM_Cubic>]}
         RedCurve {keys [<at 0 1 RCIM_Cubic>,<at 1 1 RCIM_Cubic>]}
        }}))
     B=<2 2 2>
    )
    ScaleA=true
    ScaleRGB=true
   )
  ]
  rendererClasses [&NiagaraRibbonRendererProperties]
  renderers [{rendererClass=&NiagaraRibbonRendererProperties}]
 }
 Trail001 {
  emitterUpdateScript [&EmitterState(Life Cycle Mode=::NewEnumerator1, Loop Behavior=::NewEnumerator1, Loop Duration=2.5, UseLoopDelay=false)]
  particleSpawnScript [
   &InitializeParticle(
    Color=<rgba 15 15 15 1>
    Lifetime=0.5
    Module.Write Color=true
    Module.Write Lifetime=true
    Module.Write Mass=true
    Module.Write Material Random=true
    Module.Write Position=true
    Module.Write Scale=false
    Module.Write SpriteRotation=false
    Module.Write SpriteSize=true
    Position=&Add_Vector(A=&SimulationPosition(), B=<0.5 0 0>)
    Sprite Size=&Vector2DFromFloat(Value=20)
   ) {enabled=false}
   &InitializeRibbon(
    Lifetime=&UniformRangedFloat(Minimum=0.3, Maximum=0.4)
    Module.Write Color=false
    Module.Write Lifetime=true
    Module.Write Mass=true
    Module.Write Material Random=true
    Module.Write Position=true
    Module.Write RibbonTwist=false
    Module.Write RibbonWidth=true
    Ribbon Width=10
   )
  ]
  particleUpdateScript [
   &SolveForcesAndVelocity(Clamp Velocity=false, Limit Acceleration=false)
   &ParticleState()
   &ScaleRibbonWidth(Ribbon Width Scale=&FloatFromCurve(FloatCurve=DataInterface() {propertyValues {Curve {keys [<at 0 1>,<at 1 0>]}}}))
   &ScaleColor(
    Color Value To Scale=$Particles.Spawn_Color
    Scale Alpha=&FloatFromCurve(FloatCurve=DataInterface() {propertyValues {Curve {keys [<at 0 0.5>,<at 1 0>]}}})
    Scale RGB=&Multiply_Vector(
     A=&MakeVectorFromLinearColorRGB(Color=&ColorFromCurve(ColorCurve=DataInterface() {propertyValues {
         AlphaCurve {keys [<at 0 1 RCIM_Cubic>,<at 1 1 RCIM_Cubic>]}
         BlueCurve {keys [<at 0 1 RCIM_Cubic>,<at 1 1 RCIM_Cubic>]}
         GreenCurve {keys [<at 0 1 RCIM_Cubic>,<at 1 1 RCIM_Cubic>]}
         RedCurve {keys [<at 0 1 RCIM_Cubic>,<at 1 1 RCIM_Cubic>]}
        }}))
     B=<2 2 2>
    )
    ScaleA=true
    ScaleRGB=true
   )
  ]
  rendererClasses [&NiagaraRibbonRendererProperties]
  renderers [{rendererClass=&NiagaraRibbonRendererProperties}]
 }
}
summary {
 systemName=NS_FrontEnd_Emote_Keeper_Dream_Hook_trail_and_Smoke
 emitters [
  {bEnabled=false, emitterName=Trail, rendererClasses [&NiagaraRibbonRendererProperties]}
  {bEnabled=false, emitterName=High_smoke, rendererClasses [&NiagaraMeshRendererProperties]}
  {bEnabled=false, emitterName=Flash_lead, rendererClasses [&NiagaraSpriteRendererProperties]}
  {bEnabled=false, emitterName=Round_smoke, rendererClasses [&NiagaraSpriteRendererProperties]}
  {emitterName=Star_hand, rendererClasses [&NiagaraSpriteRendererProperties]}
  {emitterName=Glow_leftHand, rendererClasses [&NiagaraSpriteRendererProperties]}
  {emitterName=Glow_RightHand, rendererClasses [&NiagaraSpriteRendererProperties]}
  {bEnabled=false, emitterName=Burst_Flash, rendererClasses [&NiagaraSpriteRendererProperties]}
  {bEnabled=false, emitterName=High_smoke_shape_more, rendererClasses [&NiagaraMeshRendererProperties]}
  {emitterName=Trail001, rendererClasses [&NiagaraRibbonRendererProperties]}
  {emitterName=High_smoke001, rendererClasses [&NiagaraMeshRendererProperties]}
  {emitterName=Flash_lead001, rendererClasses [&NiagaraSpriteRendererProperties]}
  {emitterName=Round_smoke001, rendererClasses [&NiagaraSpriteRendererProperties]}
  {emitterName=Burst_Flash001, rendererClasses [&NiagaraSpriteRendererProperties]}
  {emitterName=High_smoke_shape_more001, rendererClasses [&NiagaraMeshRendererProperties]}
  {emitterName=Burst_Sparkles, rendererClasses [&NiagaraSpriteRendererProperties]}
 ]
 userVariables [
  User.bIsFrontEnd {type=NiagaraBool}
  User.bIsFrontEndPreview {type=NiagaraBool}
  User.SkeletalMesh {
   type=NiagaraDataInterfaceSkeletalMesh
   defaultValue {
    struct=NiagaraExt_VariableValue_DataInterface
    value {dataInterface=&self:NiagaraDataInterfaceSkeletalMesh_0}
   }
  }
 ]
}
